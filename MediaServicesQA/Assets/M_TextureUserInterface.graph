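# M_TextureUserInterface is a User Interface domain material in MediaServicesQA, apparently for showing media textures in UI widgets.
# It holds a single Texture2D sample parameter, MediaTexture, defaulting to MT_PIPVideoTexture, so material instances can swap in another texture.
# The graph lists no connection from that sample to a material output and no other nodes.
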Material /MediaServicesQA/Assets/M_TextureUserInterface
graph {
 properties {materialDomain=MD_UI, pixelDepthOffsetMode=PDOM_Legacy, usedShadingModels=Default Lit}
 nodes [TextureSampleParameter2D() {parameterName=MediaTexture, texture=&/MT_PIPVideoTexture}]
}
